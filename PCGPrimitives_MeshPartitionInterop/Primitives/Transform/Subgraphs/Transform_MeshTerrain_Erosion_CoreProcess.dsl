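; PCG subgraph that runs the core erosion step for mesh terrain on point data.
; It saves each point's position to a temporary attribute, then applies a Custom HLSL GPU erosion driven by octave, scale, gain, lacunarity, strength, slope and branch parameters.
; A switch on the MaskAttribute string check either lerps original to eroded position by that mask or passes the eroded result through.
; Temporary double-underscore attributes are deleted before output.
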(graph Transform_MeshTerrain_Erosion_CoreProcess
  (params (ErosionOctaves :type Int :default 5) (ErosionScale :type Double :default 5e+03) (ErosionGain :type Double :default 0.5) (ErosionLacunarity :type Double :default 2.0) (ErosionStrength :type Double :default 5e+02) (ErosionSlopeStrength :type Double :default 4.0) (ErosionBranchStrength :type Double :default 3.0) (MaskAttribute :type String :default "$Density"))
  (bind FilterDataByType_0 (Filter%20Data%20By%20Type :TargetType (struct :Ids [(struct :Struct /Script/PCG.PCGDataTypeInfoPoint)])))
  ;[0 0 0 0.5] Core Process
  (bind CopyAttributes_0 (Copy%20Attributes :InputSource $Position :OutputTarget __Position))
  (bind CustomHLSL_1 (Custom%20HLSL :InputPins [(struct :Label In :AllowedTypes (struct :Ids [(struct :Struct /Script/PCG.PCGDataTypeInfoPoint)]) :PinStatus Required :bAllowMultipleConnections false) (struct :Label ErosionOctaves :AllowedTypes (struct :Ids [(struct :Struct /Script/PCG.PCGDataTypeInfoParam)]) :bAllowMultipleConnections false) (struct :Label ErosionScale :AllowedTypes (struct :Ids [(struct :Struct /Script/PCG.PCGDataTypeInfoParam)]) :bAllowMultipleConnections false) (struct :Label ErosionGain :AllowedTypes (struct :Ids [(struct :Struct /Script/PCG.PCGDataTypeInfoParam)]) :bAllowMultipleConnections false) (struct :Label ErosionLacunarity :AllowedTypes (struct :Ids [(struct :Struct /Script/PCG.PCGDataTypeInfoParam)]) :bAllowMultipleConnections false) (struct :Label ErosionStrength :AllowedTypes (struct :Ids [(struct :Struct /Script/PCG.PCGDataTypeInfoParam)]) :bAllowMultipleConnections false) (struct :Label ErosionSlopeStrength :AllowedTypes (struct :Ids [(struct :Struct /Script/PCG.PCGDataTypeInfoParam)]) :bAllowMultipleConnections false) (struct :Label ErosionBranchStrength :AllowedTypes (struct :Ids [(struct :Struct /Script/PCG.PCGDataTypeInfoParam)]) :bAllowMultipleConnections false)] :OutputPins [(struct :PropertiesGPU (struct :PinsToInititalizeFrom [In]) :Label Out :AllowedTypes (struct :Ids [(struct :Struct /Script/PCG.PCGDataTypeInfoPoint)]))] :ErosionBranchStrength ErosionBranchStrength :ErosionGain ErosionGain :ErosionLacunarity ErosionLacunarity :ErosionOctaves ErosionOctaves :ErosionScale ErosionScale :ErosionSlopeStrength ErosionSlopeStrength :ErosionStrength ErosionStrength)) ; Erosion effect
  (bind Switch_0 (Switch :EnumSelection (struct :Class /Script/ACLPlugin.ACLCompressionLevel))) ; Use Mask or not
  (bind AttributeMathsOp_0 (Attribute%20Maths%20Op :InputSource1 __Position :InputSource2 $Position :Operation Lerp :OutputTarget $Position :InputSource3-param MaskAttribute)) ; Lerp Position using MaskAttribute
  (bind DeleteAttributes_0 (Delete%20Attributes :Operator Matches :SelectedAttributes "__*")) ; Delete temporary attribute
  ;;[0.15 0.15 0.15 0.5] Use MaskAttribute true/false
  (bind AttributeSetToPoint_1 (Attribute%20Set%20To%20Point :In MaskAttribute))
  (bind AttributeFilter_1 (Filter%20Attribute%20Elements :AttributeTypes (struct :Type String) :bUseConstantThreshold true :Operator Matches :TargetAttribute MaskAttribute :ThresholdAttribute @Last))
  (bind GetElementsCount_1 (Get%20Element%20Count))
  (connect AttributeSetToPoint_1 Out AttributeFilter_1 In)
  (connect Switch_0 %30 AttributeMathsOp_0 InA)
  (connect CustomHLSL_1 Out AttributeMathsOp_0 InB)
  (connect CustomHLSL_1 Out AttributeMathsOp_0 Ratio)
  (connect FilterDataByType_0 InsideFilter CopyAttributes_0 Source)
  (connect FilterDataByType_0 InsideFilter CopyAttributes_0 Target)
  (connect CopyAttributes_0 Out CustomHLSL_1 In)
  (connect AttributeMathsOp_0 Out DeleteAttributes_0 In)
  (connect Switch_0 Default DeleteAttributes_0 In)
  (connect DefaultInputNode In FilterDataByType_0 In)
  (connect AttributeFilter_1 InsideFilter GetElementsCount_1 In)
  (connect CustomHLSL_1 Out Switch_0 In)
  (connect GetElementsCount_1 Out Switch_0 IntegerSelection)
  (output Out DeleteAttributes_0))
